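# Physics asset for the ConfectionPop backpack cosmetic, defining the bodies and joints for its dangling ears, arms, feet, nose and tail.
# Each dyn_ bone gets a single capsule body, with a large capsule on dyn_boby_backpack as the hub.
# Ears form two-link chains and the tail a four-link chain from that hub, while arms, feet and nose attach directly.
# Constraints use tight swing and twist limits, with nose twist locked.
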PhysicsAsset /BRCosmetics/Accessories/FORT_Backpacks/Backpack_F_MED_ConfectionPop/Meshes/F_MED_ConfectionPop_Pack_Physics
bodies {
 dyn_L_arm0 [{shapeType=Capsule, length=2.1675, radius=0.8325, shapeName=dyn_L_arm0_capsule, center=<0.1636 0 -0.1723>, rotation=<rot -46.4876 0 -0.0001>}]
 dyn_L_ear0 [{shapeType=Capsule, length=3.19, radius=1.3329, shapeName=dyn_L_ear0_capsule, center=<0.6353 0 -0.2586>, rotation=<rot -22.1515 -0.0001 0>}]
 dyn_L_ear1 [{shapeType=Capsule, length=2.4008, radius=1.7403, shapeName=dyn_L_ear1_capsule, rotation=<rot 0.0001 89.9249 -5.8379>}]
 dyn_L_foot0 [{shapeType=Capsule, length=0.9749, radius=1.5264, shapeName=dyn_L_foot0_capsule}]
 dyn_R_arm0 [{shapeType=Capsule, length=2.1675, radius=0.8325, shapeName=dyn_R_arm0_capsule, center=<0.1721 0 -0.155>, rotation=<rot -42.0184 0 0.0001>}]
 dyn_R_ear0 [{shapeType=Capsule, length=3.19, radius=1.3329, shapeName=dyn_R_ear0_capsule, center=<0.7892 0 -0.3701>, rotation=<rot -25.1245 0.0001 0>}]
 dyn_R_ear1 [{shapeType=Capsule, length=2.6967, radius=1.47, shapeName=dyn_R_ear1_capsule, rotation=<rot -0.0001 89.3974 -5.3023>}]
 dyn_R_foot0 [{shapeType=Capsule, length=0.7316, radius=1.6071, shapeName=dyn_R_foot0_capsule}]
 dyn_boby_backpack [{shapeType=Capsule, length=1.8795, radius=7.4479, center=<0.8736 -0.3376 0.1472>}]
 dyn_nose0 [{shapeType=Capsule, length=0.505, radius=0.505, shapeName=dyn_nose0_capsule, center=<0.3486 -0.0001 0>, rotation=<rot 0 -0.001 -89.9749>}]
 dyn_tail0 [{shapeType=Capsule, length=5.3465, radius=1.9948, shapeName=dyn_tail0_capsule, center=<2.6598 0 0>, rotation=<rot 0 0 90>}]
 dyn_tail1 [{shapeType=Capsule, length=4.7715, radius=3.082, shapeName=dyn_tail1_capsule, center=<2.7046 0 0>, rotation=<rot 1.0228 5.3851 79.279>}]
 dyn_tail2 [{shapeType=Capsule, length=4.6681, radius=1.9391, shapeName=dyn_tail2_capsule, center=<2.6166 0 0>, rotation=<rot 0 0 90>}]
 dyn_tail3 [{shapeType=Capsule, length=0, radius=2.7728, shapeName=dyn_tail3_capsule, center=<2.5149 0 0>}]
}
constraints [
 {bone1Name=dyn_L_ear0, bone2Name=dyn_boby_backpack, swing1LimitDegrees=30, swing2LimitDegrees=30, twistLimitDegrees=3}
 {bone1Name=dyn_L_ear1, bone2Name=dyn_L_ear0, swing1LimitDegrees=0, swing2LimitDegrees=10, twistLimitDegrees=3}
 {bone1Name=dyn_L_foot0, bone2Name=dyn_boby_backpack, swing1LimitDegrees=5, swing2LimitDegrees=10, twistLimitDegrees=1}
 {bone1Name=dyn_R_ear0, bone2Name=dyn_boby_backpack, swing1LimitDegrees=15, swing2LimitDegrees=30, twistLimitDegrees=3}
 {bone1Name=dyn_R_ear1, bone2Name=dyn_R_ear0, swing1LimitDegrees=0, swing2LimitDegrees=15, twistLimitDegrees=3}
 {bone1Name=dyn_R_foot0, bone2Name=dyn_boby_backpack, swing1LimitDegrees=5, swing2LimitDegrees=10, twistLimitDegrees=1}
 {bone1Name=dyn_nose0, bone2Name=dyn_boby_backpack, swing1LimitDegrees=10, swing2LimitDegrees=10, twistLimitDegrees=45, twistMotion=Locked}
 {bone1Name=dyn_tail0, bone2Name=dyn_boby_backpack, swing1LimitDegrees=2, swing2LimitDegrees=2, twistLimitDegrees=2}
 {bone1Name=dyn_tail1, bone2Name=dyn_tail0, swing1LimitDegrees=30, swing2LimitDegrees=16, twistLimitDegrees=5}
 {bone1Name=dyn_tail2, bone2Name=dyn_tail1, swing1LimitDegrees=16, swing2LimitDegrees=15, twistLimitDegrees=3}
 {bone1Name=dyn_tail3, bone2Name=dyn_tail2, swing1LimitDegrees=15, swing2LimitDegrees=15, twistLimitDegrees=3}
 {bone1Name=dyn_R_arm0, bone2Name=dyn_boby_backpack, swing1LimitDegrees=5, swing2LimitDegrees=40, twistLimitDegrees=0}
 {bone1Name=dyn_L_arm0, bone2Name=dyn_boby_backpack, swing1LimitDegrees=5, swing2LimitDegrees=40, twistLimitDegrees=0}
]
