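# Animation Blueprint for the ConfectionPop backpack cosmetic, posing the pack and adding secondary physics motion.
# Additive pack bone offsets include adjustments for the front end and for skydiving, which also pitches the tail, before a RigidBody simulation and final root nudge.
# Each update derives skydive state and feeds a gravity override from a deprecated state function into that simulation.
# An Anim Dynamics arm node and another root transform sit unconnected.
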AnimBlueprint /BRCosmetics/Accessories/FORT_Backpacks/Backpack_F_MED_ConfectionPop/Meshes/F_MED_ConfectionPop_Pack_AnimBP
target_skeleton=&/F_MED_ConfectionPop_Pack_Skeleton
graphs [
 AnimGraph {nodes [
   Anim Dynamics - Bone: dyn_R_arm0(Alpha=1, ExternalForce="0,0,0") {
    angularDampingOverride=0.9
    angularSpringConstant=111
    bAngularSpring=true
    bOverrideAngularDamping=true
    bUseGravityOverride=true
    simulationSpace=BoneRelative
    alphaBoolBlend {blendOption=Linear}
    boundBone {boneName=dyn_R_arm0}
    componentAppliedLinearAccClamp=<11000 11000 11000>
    componentLinearAccScale=<1 1 1>
    componentLinearVelScale=<1 1 1>
    gravityOverride {z=-920}
    physicsBodyDefinitions [{
      boundBone {boneName=dyn_R_arm0}
      boxExtents=<2 2 2>
      constraintSetup {angularLimitsMax=<0 11 0>, angularLimitsMin=<0 -10 0>, angularTarget=<0 0.1 0>}
      localJointOffset=<0.1 0 0>
     }]
    relativeSpaceBone {boneName=dyn_boby_backpack}
   }
   "Transform (Modify) Bone - Bone: root"!(ComponentPose, bAlphaBoolEnabled:bool=True, Translation="0,0,0", Rotation="0,0,6", Scale) -> (Pose) {
    alphaInputType=Bool
    rotationMode=BMM_Additive
    translationMode=BMM_Additive
    translationSpace=BCS_BoneSpace
    alphaBoolBlend {blendOption=Linear}
    boneToModify {boneName=root}
    rotation {roll=6}
   }
   n1 = "Transform (Modify) Bone - Bone: pack"!(
    ComponentPose=Local To Component(LocalPose=Input Pose())
    bAlphaBoolEnabled:bool=Get bIsInFrontEnd()
    Translation="0,-1.6,-1.256"
    Rotation="0,0,6"
    Scale
   ) -> (Pose) {
    alphaInputType=Bool
    rotationMode=BMM_Additive
    rotationSpace=BCS_BoneSpace
    translationMode=BMM_Additive
    translationSpace=BCS_BoneSpace
    alphaBoolBlend {blendOption=Linear}
    boneToModify {boneName=pack}
    rotation {roll=6}
    translation {y=-1.6, z=-1.256}
   }
   n2 = "Transform (Modify) Bone - Bone: pack"(ComponentPose=@n1, Translation="0,1.137,1.256", Rotation="0,0,6") {
    alphaInputType=Bool
    rotationMode=BMM_Additive
    rotationSpace=BCS_BoneSpace
    translationMode=BMM_Additive
    translationSpace=BCS_BoneSpace
    alphaBoolBlend {blendOption=Linear}
    boneToModify {boneName=pack}
    rotation {roll=6}
    translation {y=1.137, z=1.256}
   }
   n3 = Get fullSkyDivi() -> (fullSkyDivi:bool)
   n4 = "Transform (Modify) Bone - Bone: pack"(ComponentPose=@n2, bAlphaBoolEnabled=@n3, Translation="0,-1,0", Rotation="0,0,0") {
    alphaInputType=Bool
    rotationMode=BMM_Additive
    rotationSpace=BCS_BoneSpace
    translationMode=BMM_Additive
    translationSpace=BCS_BoneSpace
    alphaBoolBlend {blendOption=Linear}
    boneToModify {boneName=pack}
    translation {y=-1}
   }
   n5 = "Transform (Modify) Bone - Bone: dyn_tail0"!(ComponentPose=@n4, bAlphaBoolEnabled:bool=@n3, Translation="0,0,0", Rotation="-10,0,0", Scale="1,1,1") -> (Pose) {
    alphaInputType=Bool
    rotationMode=BMM_Additive
    rotationSpace=BCS_BoneSpace
    translationMode=BMM_Additive
    translationSpace=BCS_BoneSpace
    alphaBoolBlend {blendOption=Linear}
    boneToModify {boneName=dyn_tail0}
    rotation {pitch=-10}
   }
   n6 = RigidBody!(ComponentPose=@n5, Alpha=1.4, OverrideWorldGravity:Vector=Get Gravity Override(), ExternalForce) -> (Pose) {
    bClampLinearTranslationLimitToRefPose=true
    bOverrideWorldGravity=true
    overridePhysicsAsset=&/F_MED_ConfectionPop_Pack_Physics
    simulationSpace=BaseBoneSpace
    alphaBoolBlend {blendOption=Linear}
    baseBoneRef {boneName=root}
    componentAppliedLinearAccClamp=<1000 1000 1000>
    componentLinearAccScale=<0.3 0.3 0.3>
    componentLinearVelScale=<0.3 0.3 0.3>
    overrideWorldGravity {z=-980}
    simSpaceSettings {
     maxAngularAcceleration=10
     maxAngularVelocity=3
     maxLinearAcceleration=400
     maxLinearVelocity=250
     velocityScaleZ=0.1
     worldAlpha=1
     externalLinearDragV=<0.1 0.1>
    }
   }
   n7 = "Transform (Modify) Bone - Bone: root"!(ComponentPose=@n6, Alpha:float, Translation="0,-1,0", Rotation="0,0,0", Scale) -> (Pose) {
    translationMode=BMM_Additive
    translationSpace=BCS_BoneSpace
    alphaBoolBlend {blendOption=Linear}
    boneToModify {boneName=root}
    translation {y=-1}
   }
   Output Pose(Result=Component To Local(ComponentPose=@n7)) {layerGroup=None}
  ]}
 EventGraph {nodes [
   Event BlueprintUpdateAnimation()
   > n1 = Gravity Override State Function DEPRECATED(
    CharacterPartAnimInstance=Self-Reference()
    GameTimeInSeconds=GetGameTimeInSeconds()
    GravityOverrideParams=Get Gravity Override Params_0() -> (Gravity Override Params_0:Gravity Override Params Struct Structure)
   )
   > Set Gravity Override(Gravity Override=@n1)
   > Set fullSkyDivi(fullSkyDivi:bool=OR Boolean!(A=Get bIsSkydiveDiveMode(), B=Get bIsSkyDiving(), C:bool=Get bIsSkydiveFloating()) -> (ReturnValue)) -> (Output_Get:bool)
   TryGetPawnOwner()
  ]}
]
variables [
 fullSkyDivi {type=boolean}
 Gravity Override {type=Vector}
 Gravity Override Params {
  type=GravityOverrideParamsStruct
  default {
   skydive_DivingDown {gravityOverride=<0 0 0>}
   skydive_FloatingDown {gravityOverride=<0 200 500>, windAmplitude=<0 100 200>, windFrequency=<0 35 35>, windOffset=<0 -400 -300>}
   skydive_GoingUp {gravityOverride=<0 -500 500>, windAmplitude=<0 12 222>, windFrequency=<22 40 22>, windOffset=<0 -400 -300>}
   skydive_Parachuting {gravityOverride=<0 10000 20000>}
  }
 }
 Gravity Override Params_0 {
  type=GravityOverrideParamsStruct
  default {
   skydive_FloatingDown {gravityOverride=<0 33 22>, windAmplitude=<0 33 111>, windFrequency=<0 7 22>, windOffset=<0 0 1>}
   skydive_GoingUp {gravityOverride=<0 33 11>, windAmplitude=<0 33 1111>, windFrequency=<1 7 5>, windOffset=<0 0 1>}
  }
 }
]
